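# Volume-domain material for a colored fog glow around an object, in the Helios FloatMax Lights prop folder.
# World position is scaled about the object position by the Stretch parameter, so the spherical falloff can be deformed into an ellipsoid.
# A soft sphere mask sized by RingSize is raised to a power to tighten the falloff, then tinted by Fog Color, red by default.
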Material /Game/Environments/Helios/Props/FloatMax/FloatMax_Lights_A/BluePrints/M_LightGlow_Fog
graph {
 properties {bAutomaticallySetUsageInEditor=false, blendMode=BLEND_Additive, materialDomain=MD_Volume, shadingModel=MSM_Unlit, usedShadingModels=Unlit}
 nodes [
  Constant() {r=1}
  n1 = ObjectPositionWS()
  n2 = WorldPosition()
  n3 = Add(A=@n1, B=Multiply(A=Subtract(A=@n2.XYZ, B=@n1), B=VectorParameter() {parameterName=Stretch, defaultValue=<rgba 1 1 1 1>}.RGB))
  n4 = ScalarParameter() {defaultValue=55, parameterName=RingSize}
  Multiply(
   A=Power(Base=SphereMask(A=@n3, B=@n1, Radius=@n4) {hardnessPercent=0}) {constExponent=3}
   B=VectorParameter() {parameterName=Fog Color, defaultValue=<rgba 0.8 0 0.0157 1>}.RGB
  )
 ]
}
